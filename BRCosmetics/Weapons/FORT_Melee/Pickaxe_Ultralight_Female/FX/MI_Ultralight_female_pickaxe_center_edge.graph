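# Translucent unlit material for a glowing center-edge effect on the Ultralight female pickaxe, enabled for Niagara mesh particles.
# Emissive multiplies time-of-day colors by particle color through an FX adjustment function.
# Opacity combines particle alpha, sine-shaped UV masks, caustic and blurred noise, with parameters blending toward a simpler mask or bottom fade.
# Optional vertex offset waves along local X using time-scrolled sine, UV masks and dynamic-parameter intensity.
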Material /BRCosmetics/Weapons/FORT_Melee/Pickaxe_Ultralight_Female/FX/MI_Ultralight_female_pickaxe_center_edge
graph {
 properties {
  allowTranslucentCustomDepthWrites=true
  bUsedWithNiagaraMeshParticles=true
  blendMode=BLEND_Translucent
  opacityMaskClipValue=0.05
  pixelDepthOffsetMode=PDOM_Legacy
  shadingModel=MSM_Unlit
  usedShadingModels=Unlit
 }
 nodes [
  n1 = Constant3Vector() {constant=<rgba 1 0 0 0>}
  n2 = TextureCoordinate()
  n3 = Multiply(A=Time(), B=ScalarParameter() {defaultValue=0.5, parameterName=WPO Sine Speed}) {constB=0.5}
  n4 = Sine(None=ComponentMask(None=Add(A=@n2, B=@n3)) {g=true})
  n5 = Sine(None=ComponentMask(None=TextureCoordinate()) {g=true}) {period=2}
  n6 = ScalarParameter() {parameterName=Remove WPO Mask}
  n7 = TextureCoordinate()
  n8 = Multiply(A=Power(Base=ComponentMask(None=@n7) {g=true}) {constExponent=1.5}) {constB=3}
  n9 = Multiply(
   A=Multiply(A=Multiply(A=@n1, B=@n4), B=LinearInterpolate(A=@n5, Alpha=@n6))
   B=LinearInterpolate(A=Saturate(None=@n8), Alpha=ScalarParameter() {parameterName=Remove WPO Mask})
  )
  n10 = ScalarParameter() {defaultValue=6, parameterName=WPO intensity}
  n11 = DynamicParameter() {defaultValue=<rgba 1 0 1 1>, paramNames [Param1, WPO Intensity, Param3, Param4]}
  LinearInterpolate(
   B=Multiply(A=Transform(None=@n9) {transformSourceType=TRANSFORMSOURCE_Local}, B=Add(A=@n10, B=@n11.WPO Intensity)) {constB=6}
   Alpha=ScalarParameter() {defaultValue=1, parameterName=Use WPO}
  )
  n12 = VectorParameter() {parameterName="Color_Morning (V3)", defaultValue=<rgba 1.5 0.7557 1.2399 0>}
  n13 = Constant() {r=1}
  n14 = VectorParameter() {parameterName="Color_Evening (V3)", defaultValue=<rgba 1 1 1 0>}
  n15 = VectorParameter() {parameterName="Color_Night (V3)", defaultValue=<rgba 1 0.9961 0.757 0>}
  n16 = ParticleColor()
  MaterialFunctionCall(Emissive=Multiply(
    A=MaterialFunctionCall(Color_Morning=@n12.RGB, Color_Day=@n13, Color_Evening=@n14.RGB, Color_Night=@n15.RGB) {materialFunction=&Color_Time_of_Day}.Result
    B=@n16.RGB
   )) {materialFunction=&MF_FX_EmissiveAdjustments}
  n17 = Sine(None=ComponentMask(None=TextureCoordinate()) {r=true}) {period=2}
  n18 = Sine(None=ComponentMask(None=TextureCoordinate()) {g=true}) {period=2}
  n19 = Multiply(A=@n16.A, B=Multiply(A=Power(Base=@n17) {constExponent=10}, B=Power(Base=@n18) {constExponent=1}))
  n20 = TextureCoordinate()
  n21 = OneMinus(None=Sine(None=ComponentMask(None=@n20) {r=true}) {period=2})
  n22 = Saturate(None=Multiply(A=Power(Base=@n21) {constExponent=3.5}) {constB=50})
  n23 = Panner(Coordinate=Add(A=TextureCoordinate() {vTiling=0.2}, B=Constant2Vector() {g=0.7, r=0.4}) {constB=0.2})
  n24 = Multiply(A=Multiply(
    A=Multiply(A=@n19, B=@n22)
    B=TextureSample(UVs=@n23) {samplerType=SAMPLERTYPE_LinearGrayscale, texture=/Game/Global/Textures/Default/Caustics/T_Nosie_Caustics_Grayscale}.R
   )) {constB=10}
  n25 = TextureSample() {samplerType=SAMPLERTYPE_Grayscale, texture=&LowResBlurredNoise}
  n26 = ParticleColor()
  n27 = ComponentMask(None=TextureCoordinate()) {r=true}
  n28 = ComponentMask(None=TextureCoordinate()) {g=true}
  n29 = Multiply(
   A=Power(Base=Sine(None=@n27) {period=2}, Exp=ScalarParameter() {defaultValue=2, parameterName=Mask_power})
   B=Power(Base=Sine(None=@n28) {period=2}, Exp=ScalarParameter() {defaultValue=1, parameterName=Mask_power02}) {constExponent=1}
  )
  n30 = OneMinus(None=ComponentMask(None=TextureCoordinate()) {g=true})
  Multiply(
   A=LinearInterpolate(A=Multiply(A=@n24, B=@n25.R), B=Multiply(A=@n26.A, B=@n29), Alpha=ScalarParameter() {parameterName=Use Simple Mask})
   B=LinearInterpolate(B=Power(Base=@n30) {constExponent=1.5}, Alpha=ScalarParameter() {parameterName=Use bottom mask}) {constA=1}
  )
 ]
}
