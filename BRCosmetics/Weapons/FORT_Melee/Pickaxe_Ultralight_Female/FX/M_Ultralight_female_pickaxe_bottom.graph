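# Translucent, unlit, two-sided Niagara mesh-particle material for the Ultralight Female pickaxe's bottom effect, creating scrolling energy smoke.
# Panning smoke and plasma plus rotating electric-noise highlights are tinted by particle color, desaturated through a dynamic parameter, then scaled by time-of-day color for emissive.
# Panning noise, sine UV edge fades, particle alpha and an optional bottom mask form opacity.
# A time-driven sine along V supplies blendable local-space vertex wobble.
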Material /BRCosmetics/Weapons/FORT_Melee/Pickaxe_Ultralight_Female/FX/M_Ultralight_female_pickaxe_bottom
graph {
 properties {
  bUsedWithNiagaraMeshParticles=true
  blendMode=BLEND_Translucent
  pixelDepthOffsetMode=PDOM_Legacy
  shadingModel=MSM_Unlit
  twoSided=true
  usedShadingModels=Unlit
 }
 nodes [
  n1 = Time()
  n2 = ComponentMask(None=Add(A=TextureCoordinate(), B=Multiply(A=@n1) {constB=0.5})) {g=true}
  n3 = Multiply(A=Multiply(A=Constant3Vector() {constant=<rgba 1 0 0.2 0>}, B=Sine(None=@n2)))
  LinearInterpolate(
   B=Multiply(A=Transform(None=@n3) {transformSourceType=TRANSFORMSOURCE_Local}, B=ScalarParameter() {defaultValue=6, parameterName=WPO intensity}) {constB=6}
   Alpha=ScalarParameter() {defaultValue=0.3, parameterName=Use WPO}
  )
  n4 = VectorParameter() {parameterName="Color_Morning (V3)", defaultValue=<rgba 1.3 0.887 1.1557 0>}
  n5 = Constant() {r=1}
  n6 = VectorParameter() {parameterName="Color_Evening (V3)", defaultValue=<rgba 1 1 1 0>}
  n7 = VectorParameter() {parameterName="Color_Night (V3)", defaultValue=<rgba 1 0.9961 0.757 0>}
  n8 = TextureCoordinate() {vTiling=0.5}
  n9 = Time()
  n10 = TextureSample(UVs=Add(A=Rotator(Coordinate=@n8, Time=@n9), B=Panner() {speedY=0.3})) {samplerType=SAMPLERTYPE_LinearColor, texture=/Game/Characters/Player/Male/Medium/Bodies/M_MED_SRDrift/Textures/T_ElectricNoise}
  n11 = Multiply(A=TextureSample(UVs=Panner() {speedY=-0.5}) {texture=/Game/Effects/Fort_Effects/Textures/Plasma/T_Plasma_Aurora_Seamless}.B)
  n12 = Multiply(A=Power(Base=Multiply(A=@n10.R, B=@n11))) {constB=100}
  n13 = ScalarParameter() {defaultValue=1, parameterName=Highlight Intensity}
  n14 = Multiply(A=TextureCoordinate(), B=ComponentMask(None=VectorParameter() {parameterName=UV Scale, defaultValue=<rgba 0.5 1 0 0>}.RGB) {g=true, r=true})
  n15 = ScalarParameter() {defaultValue=1, parameterName=Reverse speed}
  n16 = Multiply(A=Constant2Vector() {g=-0.3}, B=@n15)
  n17 = Power(Base=TextureSampleParameter2D(UVs=Panner(Coordinate=@n14, Speed=@n16) {speedY=-0.3}) {parameterName=Tex01, texture=&T_Inky_Smoke_Tile}.R) {constExponent=1}
  n18 = Constant2Vector() {g=-0.6}
  n19 = TextureSampleParameter2D(UVs=Panner(Coordinate=@n14, Speed=Multiply(A=@n18, B=@n15)) {speedY=-0.6}) {parameterName=Tex02, texture=&T_Plasma_Aurora_Seamless}
  n20 = Multiply(A=@n17, B=Multiply(A=StaticComponentMaskParameter(None=@n19.RGB) {defaultB=true, parameterName=Component Mask}) {constB=5})
  n21 = ParticleColor()
  n22 = Multiply(A=Add(A=Multiply(A=@n12, B=@n13), B=LinearInterpolate(A=@n17, B=@n20)), B=@n21.RGB)
  n23 = DynamicParameter() {defaultValue=<rgba 0 1 1 1>, paramNames [Des,Param2,Param3,Param4]}
  MaterialFunctionCall(Emissive=Multiply(
    A=MaterialFunctionCall(Color_Morning=@n4.RGB, Color_Day=@n5, Color_Evening=@n6.RGB, Color_Night=@n7.RGB) {materialFunction=&Color_Time_of_Day}.Result
    B=Desaturation(None=@n22, Fraction=@n23.Des)
   )) {materialFunction=&MF_FX_EmissiveAdjustments}
  n24 = Constant2Vector() {g=-0.5}
  n25 = TextureSampleParameter2D(UVs=Panner(Coordinate=@n14, Speed=Multiply(A=@n24, B=@n15)) {speedY=-0.5}) {parameterName=Tex03, samplerType=SAMPLERTYPE_Grayscale, texture=&LowResBlurredNoise}
  n26 = ComponentMask(None=TextureCoordinate()) {r=true}
  n27 = ComponentMask(None=TextureCoordinate()) {g=true}
  n28 = Multiply(A=Power(Base=Sine(None=@n26) {period=2}) {constExponent=3}, B=Power(Base=Sine(None=@n27) {period=2}))
  n29 = ComponentMask(None=TextureCoordinate()) {g=true}
  n30 = Multiply(A=Power(Base=OneMinus(None=@n29)) {constExponent=3}) {constB=50}
  Multiply(
   A=Multiply(A=@n21.A, B=Multiply(A=@n25.R, B=@n28))
   B=LinearInterpolate(A=Constant() {r=1}, B=Saturate(None=@n30), Alpha=ScalarParameter() {parameterName=Use bottom mask})
  )
  ScalarParameter() {defaultValue=5, parameterName=Tex02_intensity}
 ]
}
