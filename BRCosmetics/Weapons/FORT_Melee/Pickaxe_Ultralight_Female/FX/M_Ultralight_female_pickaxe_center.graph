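# Unlit translucent Niagara mesh-particle material for the center effect of the Ultralight Female pickaxe.
# Smoke-distorted panning caustics and streaks drive a curve-atlas color ramp, added to tinted dual-texture detail and scaled by particle color and time of day.
# Sine-based UV masks fade emissive and opacity toward mesh edges.
# Scrolling sine waves displace vertices along a local direction, with dynamic parameters adding displacement intensity and controlling desaturation.
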Material /BRCosmetics/Weapons/FORT_Melee/Pickaxe_Ultralight_Female/FX/M_Ultralight_female_pickaxe_center
graph {
 properties {
  bUsedWithNiagaraMeshParticles=true
  blendMode=BLEND_Translucent
  pixelDepthOffsetMode=PDOM_Legacy
  shadingModel=MSM_Unlit
  usedShadingModels=Unlit
 }
 nodes [
  n1 = VectorParameter() {parameterName=WPO direction, defaultValue=<rgba 1 0 0 1>}
  n2 = TextureCoordinate()
  n3 = Multiply(A=Time(), B=ScalarParameter() {defaultValue=0.5, parameterName=WPO Sine Speed}) {constB=0.5}
  n4 = Sine(None=ComponentMask(None=Add(A=@n2, B=@n3)) {g=true})
  n5 = Sine(None=ComponentMask(None=TextureCoordinate()) {g=true}) {period=2}
  n6 = ScalarParameter() {parameterName=Remove WPO Mask}
  n7 = TextureCoordinate()
  n8 = Multiply(A=Power(Base=ComponentMask(None=@n7) {g=true}) {constExponent=1.5}) {constB=3}
  n9 = Multiply(
   A=Multiply(A=Multiply(A=@n1.RGB, B=@n4), B=LinearInterpolate(A=@n5, Alpha=@n6))
   B=LinearInterpolate(A=Saturate(None=@n8), Alpha=ScalarParameter() {parameterName=Remove WPO Mask})
  )
  n10 = ScalarParameter() {defaultValue=6, parameterName=WPO Intensity}
  n11 = DynamicParameter() {defaultValue=<rgba 0 0 1 1>, paramNames [Des, WPO Intensity, Param3, Param4]}
  LinearInterpolate(
   B=Multiply(A=Transform(None=@n9) {transformSourceType=TRANSFORMSOURCE_Local}, B=Add(A=@n10, B=@n11.WPO Intensity)) {constB=6}
   Alpha=ScalarParameter() {defaultValue=1, parameterName=Use WPO}
  )
  LinearInterpolate(Alpha=ScalarParameter() {parameterName=Remove base detal})
  MaterialFunctionCall() {materialFunction=&Swizzle}
  n12 = VectorParameter() {parameterName="Color_Morning (V3)", defaultValue=<rgba 1.3 0.887 1.1557 0>}
  n13 = Constant() {r=1}
  n14 = VectorParameter() {parameterName="Color_Evening (V3)", defaultValue=<rgba 1 1 1 0>}
  n15 = VectorParameter() {parameterName="Color_Night (V3)", defaultValue=<rgba 1 0.9961 0.757 0>}
  n16 = ParticleColor()
  n17 = VectorParameter() {parameterName=Dark Tex color, defaultValue=<rgba 15 0.3906 0.3906 1>}
  n18 = TextureCoordinate() {uTiling=0.5, vTiling=0.5}
  n19 = TextureCoordinate() {vTiling=0.7}
  n20 = Multiply(A=TextureSample(UVs=Panner(Coordinate=@n19) {speedY=0.3}) {samplerType=SAMPLERTYPE_Grayscale, texture=&LowResBlurredNoise}.R) {constB=0.3}
  n21 = VectorParameter() {parameterName=Dark Tex02 speed, defaultValue=<rgba 0 0.7 0 0>}
  n22 = TextureSampleParameter2D(UVs=Panner(Coordinate=Add(A=@n18, B=@n20), Speed=ComponentMask(None=@n21.RGB) {g=true, r=true}) {speedY=0.7}) {parameterName=Dark_Tex02, texture=/Game/Effects/Fort_Effects/Textures/Tile/T_CloudScape_08}
  n23 = TextureCoordinate() {uTiling=0.5, vTiling=0.5}
  n24 = VectorParameter() {parameterName=Dark Tex01 speed, defaultValue=<rgba 0 0.2 0 0>}
  n25 = TextureSampleParameter2D(UVs=Panner(Coordinate=Add(A=@n23, B=@n20), Speed=ComponentMask(None=@n24.RGB) {g=true, r=true}) {speedY=0.2}) {parameterName=Dark_Tex01, texture=&T_ShieldCaustics}
  n26 = Multiply(A=Power(Base=Multiply(A=@n22.R, B=@n25.R))) {constB=5}
  n27 = Multiply(
   A=TextureCoordinate()
   B=ComponentMask(None=VectorParameter() {parameterName=Base_Tex_UV_Scale, defaultValue=<rgba 0.7 1 0 0>}.RGB) {g=true, r=true}
  )
  n28 = LinearInterpolate(
   A=@n27
   B=MaterialFunctionCall(XY=@n27) {materialFunction=&Swizzle}.YX
   Alpha=ScalarParameter() {parameterName=Use Swizzle}
  )
  n29 = Multiply(
   A=TextureCoordinate()
   B=ComponentMask(None=VectorParameter() {parameterName=Noise_UV_Scale, defaultValue=<rgba 1 0.7 0 0>}.RGB) {g=true, r=true}
  )
  n30 = Multiply(
   A=TextureSample(UVs=Panner(Coordinate=@n29) {speedY=0.5}) {texture=&T_Inky_Smoke_Tile}.R
   B=ScalarParameter() {defaultValue=0.2, parameterName=Noise_Intensity}
  ) {constB=0.2}
  n31 = Constant2Vector() {g=0.3}
  n32 = Constant2Vector() {r=0.3}
  n33 = ScalarParameter() {parameterName=Use Swizzle}
  n34 = TextureSampleParameter2D(UVs=Panner(Coordinate=Add(A=@n28, B=@n30), Speed=LinearInterpolate(A=@n31, B=@n32, Alpha=@n33)) {speedY=0.3}) {
   parameterName=Base_Tex
   samplerSource=SSM_Wrap_WorldGroupSettings
   samplerType=SAMPLERTYPE_LinearGrayscale
   texture=/Game/Global/Textures/Default/Caustics/T_Nosie_Caustics_Grayscale
  }
  n35 = Power(Base=@n34.R, Exp=ScalarParameter() {defaultValue=1, parameterName=Base_Tex_power}) {constExponent=1}
  n36 = Panner(Coordinate=TextureCoordinate() {vTiling=0.5}) {speedY=0.5}
  n37 = Power(Base=LinearInterpolate(Alpha=TextureSample(UVs=@n36) {texture=&T_Inky_Smoke_Tile}.R) {constA=0.1}) {constExponent=10}
  n38 = Multiply(
   A=Add(A=Multiply(A=@n35), B=Multiply(A=@n37) {constB=30}) {constB=0}
   B=ScalarParameter() {defaultValue=1, parameterName=Base detal intensity}
  )
  n39 = CurveAtlasRowParameter(CurveTime=@n38) {
   atlas=/Game/Athena/LinearColorCurves/A_Atlas_Turtleneck_Glider
   curve=/Game/Weapons/FORT_Melee/Pickaxe_Ultralight_Female/FX/Curve_Ultralight_female_pickaxe
   defaultValue=12
   sortPriority=0
  }
  n40 = Add(
   A=Multiply(A=Multiply(A=@n17.RGB, B=@n26), B=ScalarParameter() {defaultValue=1, parameterName=Dark_Tex_Intensity})
   B=Add(A=@n39, B=VectorParameter() {defaultValue=<rgba 1 1 1 1>}.RGB)
  )
  n41 = Sine(None=ComponentMask(None=TextureCoordinate()) {r=true}) {period=2}
  n42 = Multiply(A=@n16.RGB, B=Multiply(A=Add(A=@n40) {constB=0.1}, B=Power(Base=@n41) {constExponent=3.5}))
  n43 = DynamicParameter() {defaultValue=<rgba 0 0 1 1>, paramNames [Des, WPO Intensity, Param3, Param4]}
  MaterialFunctionCall(Emissive=Multiply(
    A=MaterialFunctionCall(Color_Morning=@n12.RGB, Color_Day=@n13, Color_Evening=@n14.RGB, Color_Night=@n15.RGB) {materialFunction=&Color_Time_of_Day}.Result
    B=Desaturation(None=@n42, Fraction=@n43.Des)
   )) {materialFunction=&MF_FX_EmissiveAdjustments}
  n44 = Add(A=@n38, B=ScalarParameter() {parameterName=Add alpha}) {constB=0}
  n45 = TextureCoordinate()
  n46 = Power(Base=Sine(None=ComponentMask(None=@n45) {r=true}) {period=2}) {constExponent=1}
  n47 = Sine(None=ComponentMask(None=TextureCoordinate()) {g=true}) {period=2}
  n48 = ScalarParameter() {defaultValue=2, parameterName=G mask power}
  Multiply(A=@n16.A, B=Multiply(A=Multiply(A=@n44, B=@n46), B=Power(Base=@n47, Exp=@n48)))
 ]
}
