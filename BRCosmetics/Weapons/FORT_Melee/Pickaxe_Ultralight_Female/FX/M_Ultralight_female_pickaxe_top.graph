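# Translucent unlit Niagara mesh-particle material for the top effect of the female Ultralight pickaxe, producing animated, glowing energy.
# Emissive layers smoke-distorted caustics, electric-plasma noise and UV color gradients, with opposite-direction panners blended along U.
# Particle color, time-of-day tint and dynamic-parameter desaturation shape the glow.
# Opacity combines panning noise, a V-axis sine mask and particle alpha, while a "Use WPO" parameter scales a local-space sine wobble.
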Material /BRCosmetics/Weapons/FORT_Melee/Pickaxe_Ultralight_Female/FX/M_Ultralight_female_pickaxe_top
graph {
 properties {
  bUsedWithNiagaraMeshParticles=true
  blendMode=BLEND_Translucent
  pixelDepthOffsetMode=PDOM_Legacy
  shadingModel=MSM_Unlit
  usedShadingModels=Unlit
 }
 nodes [
  n1 = Constant3Vector() {constant=<rgba 0 0 1 0>}
  n2 = TextureCoordinate()
  n3 = Multiply(A=Time()) {constB=0.5}
  n4 = Sine(None=ComponentMask(None=Add(A=@n2, B=@n3)) {r=true}) {period=0.8}
  n5 = ComponentMask(None=TextureCoordinate()) {g=true}
  n6 = ComponentMask(None=TextureCoordinate()) {r=true}
  n7 = Multiply(A=Multiply(A=Multiply(A=@n1, B=@n4), B=Sine(None=@n5) {period=2}), B=Power(Base=OneMinus(None=@n6)) {constExponent=1})
  LinearInterpolate(
   B=Multiply(A=Transform(None=@n7) {transformSourceType=TRANSFORMSOURCE_Local}) {constB=6}
   Alpha=ScalarParameter() {defaultValue=0.2, parameterName=Use WPO}
  )
  n8 = VectorParameter() {parameterName="Color_Morning (V3)", defaultValue=<rgba 1.5 0.7557 1.2399 0>}
  n9 = Constant() {r=1}
  n10 = VectorParameter() {parameterName="Color_Evening (V3)", defaultValue=<rgba 1 1 1 0>}
  n11 = VectorParameter() {parameterName="Color_Night (V3)", defaultValue=<rgba 1 0.9961 0.757 0>}
  n12 = VectorParameter() {parameterName=Param_1, defaultValue=<rgba 5 0.7101 0.4613 1>}
  n13 = Panner(Coordinate=TextureCoordinate()) {speedY=0.6}
  n14 = Add(
   A=TextureCoordinate() {vTiling=0.5}
   B=Multiply(A=TextureSample(UVs=@n13) {texture=&T_Inky_Smoke_Tile}.R) {constB=0.2}
  )
  n15 = Power(Base=TextureSample(UVs=Panner(Coordinate=@n14) {speedX=0.6}) {samplerType=SAMPLERTYPE_LinearGrayscale, texture=/Game/Global/Textures/Default/Caustics/T_Nosie_Caustics_Grayscale}.B)
  n16 = Multiply(A=@n12.RGB, B=Add(A=Multiply(A=@n15)) {constB=0})
  n17 = TextureCoordinate()
  n18 = Multiply(A=SmoothStep(Min=ComponentMask(None=@n17) {r=true}, Value=Constant() {r=0.65})) {constB=2}
  n19 = LinearInterpolate(Alpha=Power(Base=Saturate(None=@n18)))
  n20 = Panner(Coordinate=TextureCoordinate()) {speedY=0.6}
  n21 = Add(
   A=TextureCoordinate() {vTiling=0.5}
   B=Multiply(A=TextureSample(UVs=@n20) {texture=&T_Inky_Smoke_Tile}.R) {constB=0.2}
  )
  n22 = Power(Base=TextureSample(UVs=Panner(Coordinate=@n21) {speedX=-0.6}) {samplerType=SAMPLERTYPE_LinearGrayscale, texture=&T_Nosie_Caustics_Grayscale}.B)
  n23 = Multiply(A=Add(A=Multiply(A=@n22) {constB=2}) {constB=0}, B=@n12.RGB)
  n24 = ComponentMask(None=TextureCoordinate()) {r=true}
  n25 = ParticleColor()
  n26 = ComponentMask(None=TextureCoordinate()) {r=true}
  n27 = Constant3Vector() {constant=<rgba 0.0312 0 0 0>}
  n28 = Constant3Vector() {constant=<rgba 1 0 0 0>}
  n29 = Sine(None=ComponentMask(None=TextureCoordinate()) {r=true}) {period=2}
  n30 = Sine(None=ComponentMask(None=TextureCoordinate()) {g=true}) {period=2}
  n31 = Power(Base=OneMinus(None=Multiply(A=@n29, B=@n30))) {constExponent=3}
  n32 = ComponentMask(None=TextureCoordinate()) {r=true}
  n33 = Add(
   A=Multiply(A=LinearInterpolate(A=@n27, B=@n28, Alpha=@n31) {constA=1}, B=Power(Base=@n26))
   B=Multiply(A=Power(Base=@n32) {constExponent=5}, B=VectorParameter() {parameterName=Param_2, defaultValue=<rgba 3 0 0.6047 1>}.RGB)
  )
  n34 = TextureCoordinate() {uTiling=1.5}
  n35 = Panner(Coordinate=@n34) {speedX=-0.1}
  n36 = TextureSample(UVs=Panner(Coordinate=@n34) {speedX=-0.5}) {texture=/Game/Effects/Fort_Effects/Textures/Plasma/T_Plasma_Aurora_Seamless}
  n37 = Power(Base=Multiply(
    A=TextureSample(UVs=@n35) {samplerType=SAMPLERTYPE_LinearColor, texture=/Game/Characters/Player/Male/Medium/Bodies/M_MED_SRDrift/Textures/T_ElectricNoise}.R
    B=Multiply(A=@n36.B)
   ))
  n38 = Constant() {r=60}
  n39 = Panner(Coordinate=@n34) {speedX=0.1}
  n40 = TextureSample(UVs=Panner(Coordinate=@n34) {speedX=0.5}) {texture=&T_Plasma_Aurora_Seamless}
  n41 = Power(Base=Multiply(
    A=TextureSample(UVs=@n39) {samplerType=SAMPLERTYPE_LinearColor, texture=&T_ElectricNoise}.R
    B=Multiply(A=@n40.B)
   ))
  n42 = SmoothStep(Min=ComponentMask(None=TextureCoordinate()) {r=true}, Value=Constant() {r=0.65})
  n43 = Power(Base=Saturate(None=Multiply(A=@n42) {constB=2}))
  n44 = TextureCoordinate() {uTiling=0.5, vTiling=0.7}
  n45 = Power(Base=TextureSample(UVs=Panner(Coordinate=@n44) {speedX=-0.15}) {samplerType=SAMPLERTYPE_LinearGrayscale, texture=&T_Nosie_Caustics_Grayscale}.R) {constExponent=1}
  n46 = Multiply(A=TextureSample(UVs=Panner(Coordinate=@n44) {speedX=-0.3}) {samplerType=SAMPLERTYPE_LinearGrayscale, texture=&T_Nosie_Caustics_Grayscale}.B)
  n47 = LinearInterpolate(A=@n45, B=Multiply(A=@n45, B=Multiply(A=@n46) {constB=10}))
  n48 = TextureCoordinate() {uTiling=0.5, vTiling=0.7}
  n49 = Power(Base=TextureSample(UVs=Panner(Coordinate=@n48) {speedX=0.15}) {samplerType=SAMPLERTYPE_LinearGrayscale, texture=&T_Nosie_Caustics_Grayscale}.R) {constExponent=1}
  n50 = Multiply(A=TextureSample(UVs=Panner(Coordinate=@n48) {speedX=0.3}) {samplerType=SAMPLERTYPE_LinearGrayscale, texture=&T_Nosie_Caustics_Grayscale}.B)
  n51 = LinearInterpolate(A=@n49, B=Multiply(A=@n49, B=Multiply(A=@n50) {constB=10}))
  n52 = TextureCoordinate()
  n53 = Multiply(A=SmoothStep(Min=ComponentMask(None=@n52) {r=true}, Value=Constant() {r=0.65})) {constB=2}
  n54 = LinearInterpolate(A=Add(B=Power(Base=@n47) {constExponent=1}), B=Add(B=Power(Base=@n51) {constExponent=1}), Alpha=Power(Base=Saturate(None=@n53)))
  n55 = Add(
   A=LinearInterpolate(A=Multiply(A=@n37, B=@n38) {constB=100}, B=Multiply(A=@n41, B=@n38) {constB=100}, Alpha=LinearInterpolate(Alpha=@n43))
   B=Multiply(A=Power(Base=@n54) {constExponent=1})
  ) {constB=0}
  n56 = Add(A=Multiply(A=LinearInterpolate(A=@n23, B=@n16, Alpha=@n19), B=Power(Base=@n24)), B=Multiply(A=Add(A=@n33, B=@n55), B=@n25.RGB))
  n57 = DynamicParameter() {defaultValue=<rgba 0 1 1 1>, paramNames [Des,Param2,Param3,Param4]}
  MaterialFunctionCall(Emissive=Multiply(
    A=MaterialFunctionCall(Color_Morning=@n8.RGB, Color_Day=@n9, Color_Evening=@n10.RGB, Color_Night=@n11.RGB) {materialFunction=&Color_Time_of_Day}.Result
    B=Desaturation(None=@n56, Fraction=@n57.Des)
   )) {materialFunction=&MF_FX_EmissiveAdjustments}
  n58 = TextureCoordinate() {vTiling=0.8}
  n59 = TextureSample(UVs=Panner(Coordinate=@n58) {speedX=-0.25}) {samplerType=SAMPLERTYPE_Grayscale, texture=&LowResBlurredNoise}
  n60 = TextureSample(UVs=Panner(Coordinate=@n58) {speedX=0.25}) {samplerType=SAMPLERTYPE_Grayscale, texture=&LowResBlurredNoise}
  n61 = SmoothStep(Min=ComponentMask(None=TextureCoordinate()) {r=true}, Value=Constant() {r=0.65})
  n62 = Power(Base=Saturate(None=Multiply(A=@n61) {constB=2}))
  n63 = Sine(None=ComponentMask(None=TextureCoordinate()) {g=true}) {period=2}
  Multiply(A=@n25.A, B=Multiply(A=LinearInterpolate(A=@n59.R, B=@n60.R, Alpha=@n62), B=Power(Base=@n63) {constExponent=6}))
 ]
}
